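# Default Lit material for a cloud of cubes in CubeSurfing, animated through world position offset and shaded in animated blues with a gradient texture.
# Vertex color is decoded into per-cube centers, around which cubes spin on Z and X and scale.
# A looping fake travel offset and two-scale noise-texture height waves provide further motion.
# Per-cube hash noise is derived for variation, and a matching rotated tangent-space normal is computed.
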Material /CubeSurfing/Gameplay/Material/M_CubeCloud
graph {
 properties {bAutomaticallySetUsageInEditor=false, maxWorldPositionOffsetDisplacement=800, usedShadingModels=Default Lit}
 nodes [
  n1 = NamedRerouteUsage()
  n2 = Divide(A=@n1)
  Add(A=@n2, B=Multiply(A=ScalarParameter() {defaultValue=1, group=WPO_Animation, parameterName=WavePanningSpeed}, B=Time()))
  n3 = ComponentMask(None=NamedRerouteUsage()) {g=true, r=true}
  n4 = Divide(A=@n3, B=ScalarParameter() {defaultValue=2560, parameterName=WindUnit})
  n5 = Constant3Vector() {constant=<rgba 1 0 0 0>}
  n6 = ScalarParameter() {defaultValue=1, group=WPO_Animation, parameterName=WavePanningSpeed}
  n7 = Time()
  Add(A=@n4, B=Multiply(A=ComponentMask(None=@n5) {g=true, r=true}, B=Multiply(A=@n6, B=@n7)))
  n8 = Transform(None=Constant3Vector() {constant=<rgba 1 0 0 0>}) {transformSourceType=TRANSFORMSOURCE_Local}
  n9 = Divide(
   A=TransformPosition(None=NamedRerouteUsage()) {transformType=TRANSFORMPOSSOURCE_World}
   B=ScalarParameter() {defaultValue=1024, group=WPO_Animation, parameterName=WorldRadius}
  )
  n10 = DotProduct(A=@n8, B=@n9)
  n11 = Constant3Vector() {constant=<rgba 0 0 1 0>}
  AppendVector(A=@n10, B=DotProduct(A=@n9, B=CrossProduct(A=@n8, B=@n11)))
  Constant3Vector() {constant=<rgba 0 0 1 0>}
  n12 = Sine(None=@n2)
  ConstantBiasScale(None=@n12)
  Constant() {r=0.5}
  Constant() {r=0.2}
  Constant()
  Constant()
  n13 = TextureSampleParameter2D() {group=Color_Shading, parameterName=CubeGradient, texture=/CubeSurfing/Assets/Textures/T_CubeGradient}
  n14 = VectorParameter() {group=Color_Shading, parameterName=DarkBlue, defaultValue=<rgba 0.0261 0.2041 0.5573 1>}
  n15 = VectorParameter() {group=Color_Shading, parameterName=LightBlue, defaultValue=<rgba 0 0.653 0.7396 1>}
  n16 = Multiply(A=NamedRerouteUsage(), B=ScalarParameter() {defaultValue=1, group=WPO_Animation, parameterName=LocalNoise_Intensity_Color}) {constB=0.3}
  n17 = Multiply(A=Time(), B=ScalarParameter() {defaultValue=0.7, group=WPO_Animation, parameterName=PanningSpeed_Color})
  n18 = ConstantBiasScale(None=Sine(None=Add(A=@n16, B=@n17)))
  n19 = Power(Base=Saturate(None=Add(A=@n18) {constB=0.2})) {constExponent=2.2}
  Multiply(A=Multiply(A=LinearInterpolate(Alpha=@n13.R) {constA=0.5}, B=LinearInterpolate(A=@n14.RGB, B=@n15.RGB, Alpha=@n19)), B=NamedRerouteUsage())
  n20 = ScalarParameter() {defaultValue=128, group=WPO_Animation, parameterName=HeightNoise_Wind}
  Multiply(A=@n12, B=@n20)
  Multiply(A=@n20, B=MaterialFunctionCall() {materialFunction=&ObjectScale}.Scale Z)
  n21 = NamedRerouteUsage()
  n22 = LinearInterpolate(
   A=ScalarParameter() {defaultValue=0.6, group=WPO_Rotation, parameterName=MaxSpeed_Z}
   B=ScalarParameter() {defaultValue=0.3, group=WPO_Rotation, parameterName=MinSpeed_Z}
   Alpha=@n21
  )
  n23 = Time()
  NamedRerouteDeclaration(None=Add(A=Multiply(A=@n21) {constB=360}, B=Multiply(A=@n22, B=@n23))) {name=Z_Degree, nodeColor=<rgba 1 0.5625 0 1>}
  n24 = NamedRerouteUsage()
  n25 = LinearInterpolate(
   A=ScalarParameter() {defaultValue=1.5, group=WPO_Rotation, parameterName=MaxSpeed_X}
   B=ScalarParameter() {defaultValue=0.8, group=WPO_Rotation, parameterName=MinSpeed_X}
   Alpha=@n24
  )
  n26 = Multiply(A=Time()) {constB=-1}
  NamedRerouteDeclaration(None=Add(A=Multiply(A=@n24) {constB=180}, B=Multiply(A=@n25, B=@n26))) {name=X_Degree, nodeColor=<rgba 1 0.5625 0 1>}
  n27 = Constant3Vector() {constant=<rgba -1 0 0 0>}
  n28 = NamedRerouteUsage()
  n29 = Time()
  n30 = ScalarParameter() {defaultValue=10, group=WPO_Animation, parameterName=TavelSpeed}
  n31 = NamedRerouteUsage()
  n32 = Add(A=Add(A=Multiply(A=@n29, B=@n30), B=NamedRerouteUsage()), B=@n31)
  n33 = NamedRerouteDeclaration(None=Frac(None=StaticSwitchParameter(True=@n28, False=@n32) {defaultValue=true, group=WPO_Animation, parameterName=UseWorldGradient})) {name=TravelPhase, nodeColor=<rgba 1 0.7031 0 1>}
  n34 = Multiply(A=ScalarParameter() {defaultValue=70, group=WPO_Animation, parameterName=TravelDist}, B=Add(A=OneMinus(None=@n31), B=NamedRerouteUsage()))
  NamedRerouteDeclaration(None=Multiply(A=Transform(None=@n27) {transformSourceType=TRANSFORMSOURCE_Local}, B=Multiply(A=@n33, B=@n34))) {name=Fake_Travel, nodeColor=<rgba 0 0.1328 1 1>}
  n35 = VertexColor()
  n36 = Multiply(
   A=Subtract(
    A=MaterialFunctionCall(X=@n35.R, Y=@n35.R, Z=@n35.R) {materialFunction=&MakeFloat3}.Result
    B=Constant3Vector() {constant=<rgba 0.5 0.5 0.5 0>}
   ) {constB=0.5}
   B=ScalarParameter() {defaultValue=1536, group=WPO_Animation, parameterName=WorldUnit}
  ) {constB=255}
  n37 = Subtract(
   A=TransformPosition(None=Constant3Vector()) {transformType=TRANSFORMPOSSOURCE_World}
   B=TransformPosition(None=@n36) {transformType=TRANSFORMPOSSOURCE_World}
  )
  n38 = TransformPosition(None=Add(A=@n37, B=WorldPosition().XYZ)) {transformSourceType=TRANSFORMPOSSOURCE_World}
  n39 = NamedRerouteUsage()
  n40 = Cosine(None=@n39)
  n41 = Sine(None=@n39)
  n42 = Constant()
  n43 = MaterialFunctionCall(X=@n40, Y=Multiply(A=@n41) {constB=-1}, Z=@n42) {materialFunction=&MakeFloat3}
  n44 = MaterialFunctionCall(X=@n41, Y=@n40, Z=@n42) {materialFunction=&MakeFloat3}
  n45 = Constant3Vector() {constant=<rgba 0 0 1 0>}
  n46 = Constant3Vector()
  n47 = Constant()
  n48 = NamedRerouteUsage()
  n49 = Cosine(None=@n48)
  n50 = Sine(None=@n48)
  n51 = Multiply(A=@n50) {constB=-1}
  n52 = OneMinus(None=NamedRerouteUsage())
  n53 = Multiply(
   A=MaterialFunctionCall(
    VectorToTransform=MaterialFunctionCall(VectorToTransform=@n38, BasisX=@n43.Result, BasisY=@n44.Result, BasisZ=@n45, BasisPosition=@n46) {materialFunction=&Transform3x3Matrix}.Transformed Vector
    BasisX=Constant3Vector() {constant=<rgba 1 0 0 0>}
    BasisY=MaterialFunctionCall(X=@n47, Y=@n49, Z=@n51) {materialFunction=&MakeFloat3}.Result
    BasisZ=MaterialFunctionCall(X=@n47, Y=@n50, Z=@n49) {materialFunction=&MakeFloat3}.Result
    BasisPosition=Constant3Vector()
   ) {materialFunction=&Transform3x3Matrix}.Transformed Vector
   B=LinearInterpolate(A=SmoothStep(Value=@n52) {constMax=0.5}, Alpha=NamedRerouteUsage())
  )
  n54 = Subtract(A=Subtract(A=TransformPosition(None=@n53) {transformType=TRANSFORMPOSSOURCE_World}, B=WorldPosition().XYZ), B=@n37)
  n55 = NamedRerouteUsage()
  NamedRerouteDeclaration(None=Add(A=Add(A=@n54, B=@n55), B=NamedRerouteUsage())) {name=WPO_OUT, nodeColor=<rgba 0.4531 1 0 1>}
  n56 = Multiply(A=NamedRerouteUsage(), B=ScalarParameter() {defaultValue=1, parameterName=LocalNoiseIntensity}) {constB=0.5}
  NamedRerouteDeclaration(None=Frac(None=Add(A=@n10, B=@n56))) {name=WorldGradient_X, nodeColor=<rgba 1 0 0.1641 1>}
  NamedRerouteDeclaration(None=@n35.A) {name=Vtx_Alpha, nodeColor=<rgba 0 1 0.3203 1>}
  n57 = OneMinus(None=@n1)
  n58 = TextureObject() {samplerType=SAMPLERTYPE_Masks, texture=/Game/Global/Textures/Default/Noise/WindNoise_Mask_1}
  n59 = Multiply(
   A=Subtract(A=TextureSample(UVs=@n4, Tex=@n58) {samplerType=SAMPLERTYPE_Masks, texture=&WindNoise_Mask_1}.R) {constB=0.2}
   B=ScalarParameter() {defaultValue=128, group=WPO_Animation, parameterName=HeightNoise_Wind}
  )
  n60 = Divide(A=@n3, B=ScalarParameter() {defaultValue=5120, parameterName=GlobalUnit})
  n61 = Multiply(
   A=Subtract(A=TextureSample(UVs=@n60, Tex=@n58) {samplerType=SAMPLERTYPE_Masks, texture=&WindNoise_Mask_1}.RGB) {constB=0.5}
   B=ScalarParameter() {defaultValue=128, parameterName=HeightNoise_Global}
  )
  n62 = Multiply(A=Constant3Vector() {constant=<rgba 0 0 1 0>}, B=Multiply(A=LinearInterpolate(Alpha=@n57) {constA=0.5}, B=Add(A=@n59, B=@n61)))
  NamedRerouteDeclaration(None=@n62) {name=HeightWave, nodeColor=<rgba 0 1 0.1562 1>}
  n63 = Constant3Vector()
  NamedRerouteDeclaration(None=Add(A=@n36, B=TransformPosition(None=@n63) {transformType=TRANSFORMPOSSOURCE_World})) {name=CenterWorld, nodeColor=<rgba 0 1 0.5547 1>}
  NamedRerouteDeclaration(None=@n36) {name=LocalWorld, nodeColor=<rgba 0.8047 1 0 1>}
  n64 = Frac(None=Length(None=NamedRerouteUsage()) {is=class MaterialExpressionLength})
  n65 = Time()
  n66 = ScalarParameter() {defaultValue=1, group=Hash, parameterName=DerezSpeed}
  n67 = Frac(None=Add(A=Multiply(A=@n64) {constB=3}, B=Multiply(A=@n65, B=@n66)))
  NamedRerouteDeclaration(None=LinearInterpolate(Alpha=Power(Base=@n67)) {constA=0.2}) {name=SimpleHashNoise, nodeColor=<rgba 0 1 0.9531 1>}
  NamedRerouteDeclaration(None=@n64) {name=LocalNoise, nodeColor=<rgba 1 0.8203 0 1>}
  n68 = Transform(None=Constant3Vector() {constant=<rgba 0 0 1 0>})
  n69 = NamedRerouteUsage()
  n70 = Cosine(None=@n69)
  n71 = Sine(None=@n69)
  n72 = Multiply(A=@n71) {constB=-1}
  n73 = Constant()
  n74 = Constant()
  n75 = NamedRerouteUsage()
  n76 = Cosine(None=@n75)
  n77 = Sine(None=@n75)
  n78 = MaterialFunctionCall(
   VectorToTransform=MaterialFunctionCall(
    VectorToTransform=Normalize(VectorInput=@n68)
    BasisX=MaterialFunctionCall(X=@n70, Y=@n72, Z=@n73) {materialFunction=&MakeFloat3}.Result
    BasisY=MaterialFunctionCall(X=@n71, Y=@n70, Z=@n73) {materialFunction=&MakeFloat3}.Result
    BasisZ=Constant3Vector() {constant=<rgba 0 0 1 0>}
   ) {materialFunction=&Transform3x3Matrix}.Transformed Vector
   BasisX=Constant3Vector() {constant=<rgba 1 0 0 0>}
   BasisY=MaterialFunctionCall(X=@n74, Y=@n76, Z=Multiply(A=@n77) {constB=-1}) {materialFunction=&MakeFloat3}.Result
   BasisZ=MaterialFunctionCall(X=@n74, Y=@n77, Z=@n76) {materialFunction=&MakeFloat3}.Result
   BasisPosition=Constant3Vector()
  ) {materialFunction=&Transform3x3Matrix}
  NamedRerouteDeclaration(None=Transform(None=Normalize(VectorInput=@n78.Transformed Vector)) {transformSourceType=TRANSFORMSOURCE_World, transformType=TRANSFORM_Tangent}) {name=TangentN, nodeColor=<rgba 1 0.1875 0 1>}
  NamedRerouteUsage()
  n79 = NamedRerouteUsage()
  PreviousFrameSwitch(Current Frame=@n79, Previous Frame=@n79)
  Transform(None=@n62) {transformSourceType=TRANSFORMSOURCE_Local}
  Transform(None=@n5) {transformSourceType=TRANSFORMSOURCE_Local}
 ]
}
