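# Translucent two-sided CubeSurfing material for a temporary derez effect, flagged for skeletal meshes and mesh deformers.
# A blue character-derez material function, fed a Fresnel rim, supplies base attributes; its remapped opacity times a UV-centered sphere mask drives opacity and opacity mask.
# Two world-XY noise samples at different scales create vertical vertex offset, scaled by normalized local height.
# The time-based panning nodes are unconnected, so that offset is static.
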Material /CubeSurfing/Gameplay/Material/M_DerezeTemp
graph {
 properties {
  bAutomaticallySetUsageInEditor=false
  bUseMaterialAttributes=true
  bUsedWithMeshDeformer=true
  bUsedWithSkeletalMesh=true
  blendMode=BLEND_Translucent
  twoSided=true
  usedShadingModels=Default Lit
 }
 nodes [
  n1 = TextureCoordinate()
  Frac(None=ComponentMask(None=@n1) {r=true})
  Multiply(A=ScalarParameter() {defaultValue=1, group=WPO_Animation, parameterName=WavePanningSpeed}, B=Time())
  n2 = Saturate(None=Divide(A=LocalPosition().Z, B=ScalarParameter() {defaultValue=384, parameterName=MaxHeight}))
  Power(Base=@n2) {constExponent=2.2}
  ScalarParameter() {defaultValue=1, group=WPO_Animation, parameterName=WavePanningSpeed}
  n3 = MaterialFunctionCall!(
   BaseMaterial=MakeMaterialAttributes!(
    BaseColor
    Metallic
    Specular
    Roughness
    Anisotropy
    EmissiveColor
    Opacity
    OpacityMask
    Normal
    Tangent
    WorldPositionOffset
    SubsurfaceColor
    ClearCoat
    ClearCoatRoughness
    AmbientOcclusion
    Refraction
    CustomizedUVs_0
    CustomizedUVs_1
    CustomizedUVs_2
    CustomizedUVs_3
    CustomizedUVs_4
    CustomizedUVs_5
    CustomizedUVs_6
    CustomizedUVs_7
    PixelDepthOffset
    ShadingModel
    Displacement
   )
   Rim=Fresnel(Normal=PixelNormalWS()) {baseReflectFraction=0, exponent=1}
  ) {materialFunction=&/MF_CharacterDerez_BaseBlue_TempRest}
  n4 = Multiply(
   A=LinearInterpolate(Alpha=@n3.Opacity) {constA=0.5}
   B=SphereMask(
    A=@n1
    B=Constant2Vector() {g=0.5, r=0.5}
    Radius=ScalarParameter() {defaultValue=0.8, parameterName=Radius}
    Hardness=ScalarParameter() {defaultValue=0.2, parameterName=Hardness}
   ) {attenuationRadius=1, hardnessPercent=0}
  )
  n5 = Constant3Vector() {constant=<rgba 0 0 1 0>}
  n6 = ComponentMask(None=WorldPosition().XYZ) {g=true, r=true}
  n7 = TextureObject() {samplerType=SAMPLERTYPE_Masks, texture=/Game/Global/Textures/Default/Noise/WindNoise_Mask_1}
  n8 = TextureSample(UVs=Divide(A=@n6, B=ScalarParameter() {defaultValue=2560, parameterName=WindUnit}), Tex=@n7) {samplerType=SAMPLERTYPE_Masks, texture=&WindNoise_Mask_1}
  n9 = TextureSample(UVs=Divide(A=@n6, B=ScalarParameter() {defaultValue=5120, parameterName=GlobalUnit}), Tex=@n7) {samplerType=SAMPLERTYPE_Masks, texture=&WindNoise_Mask_1}
  n10 = Add(
   A=Multiply(A=Subtract(A=@n8.R) {constB=0.2}, B=ScalarParameter() {defaultValue=128, group=WPO_Animation, parameterName=HeightNoise_Wind})
   B=Multiply(A=Subtract(A=@n9.RGB) {constB=0.5}, B=ScalarParameter() {defaultValue=128, parameterName=HeightNoise_Global})
  )
  SetMaterialAttributes!(
   MaterialAttributes=@n3.Result
   Opacity Mask=@n4
   Opacity=@n4
   World Position Offset=Multiply(A=Multiply(A=@n5, B=@n10), B=SmoothStep(Value=@n2) {constMax=0.2})
  ) {attributeSetTypes [#1, #2, #3]}
 ]
}
