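# Default Lit material for the magic cube mesh, animating it through world position offset with idle, burst, forming, deploying and hide motions.
# Vertex colour supplies positions and seeds, while UV1 staggers progress as pieces rotate, scale and collapse toward the object position.
# Hand-built rotation matrices drive offsets and normals.
# Blue noise-driven shading adds centre glow and deploy-timed emissive compensated for eye adaptation; an AsPickupMesh switch overrides offset, normal and brightness.
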Material /CubeSurfing/Gameplay/Material/M_MagicCubeMovement
graph {
 properties {
  bAutomaticallySetUsageInEditor=false
  maxWorldPositionOffsetDisplacement=300
  previewMesh=/CubeSurfing/Gameplay/Mesh/S_MagicCube_V02
  usedShadingModels=Default Lit
 }
 nodes [
  n1 = NamedRerouteUsage()
  n2 = Subtract(A=@n1, B=VectorParameter() {group=WPO_Animation, parameterName=BboxMin, defaultValue=<rgba 0 0 0 1>}.RGB)
  n3 = VectorParameter() {group=WPO_Animation, parameterName=BBoxRatio, defaultValue=<rgba 1 1 1 1>}
  n4 = NamedRerouteUsage()
  n5 = Multiply(
   A=Divide(A=Multiply(A=@n2, B=@n3.RGB), B=ScalarParameter() {defaultValue=5, group=WPO_Animation, parameterName=LocalUnit})
   B=Multiply(A=@n4, B=VectorParameter() {group=WPO_Animation, parameterName=AlignRatio, defaultValue=<rgba 1 1 1 1>}.RGB)
  )
  n6 = ScalarParameter() {defaultValue=10, group=WPO_Animation, parameterName=RoundActurate}
  n7 = Divide(A=Round(None=Multiply(A=@n5, B=@n6)), B=@n6)
  n8 = VectorParameter() {parameterName=FowardVector, defaultValue=<rgba 1 1 0 1>}
  n9 = DotProduct(A=@n7, B=@n8.RGB)
  n10 = DotProduct(A=@n7, B=CrossProduct(A=@n8.RGB, B=Constant3Vector() {constant=<rgba 0 0 1 0>}))
  n11 = Multiply(
   A=AppendVector(A=AppendVector(A=@n9, B=@n10), B=ComponentMask(None=@n7) {b=true})
   B=VectorParameter() {parameterName=AlignWorld, defaultValue=<rgba 1 1 0 1>}.RGB
  )
  Add(A=ActorPositionWS(), B=@n11)
  n12 = Multiply(A=NamedRerouteUsage(), B=ScalarParameter() {defaultValue=1, parameterName=LocalNoise_Intensity}) {constB=0.3}
  n13 = MaterialFunctionCall(Float3=NamedRerouteUsage()) {materialFunction=&BreakOutFloat3Components}
  n14 = Add(A=Add(A=Multiply(A=@n13.R) {constB=0.2}, B=Multiply(A=@n13.G) {constB=0.2}), B=Multiply(A=@n13.B) {constB=0.1})
  n15 = Time()
  n16 = Sine(None=Add(A=Add(A=@n12, B=@n14), B=Multiply(A=@n15, B=@n14)))
  n17 = Saturate(None=Add(A=ConstantBiasScale(None=@n16)) {constB=0.2})
  Add(A=MaterialFunctionCall(Seed=NamedRerouteUsage(), Float=@n17) {materialFunction=&MF_Hash}.Float) {constB=0}
  Constant()
  Constant() {r=0.5}
  n18 = Multiply(
   A=NamedRerouteUsage()
   B=LinearInterpolate(Alpha=TextureSampleParameter2D() {group=Color_Shading, parameterName=CubeGradient, texture=/CubeSurfing/Assets/Textures/T_CubeGradient}.R) {constA=0.5}
  )
  n19 = NamedRerouteUsage()
  n20 = NamedRerouteUsage()
  n21 = VectorParameter() {group=Color_Shading, parameterName=LightBlue_CenterGlow, defaultValue=<rgba 0.2781 0.6856 0.7396 2>}
  n22 = Multiply(A=Multiply(A=Power(Base=@n19) {constExponent=2.2}, B=OneMinus(None=@n20)), B=Multiply(A=@n21.RGB, B=@n21.RGB))
  n23 = Constant() {r=1}
  n24 = Divide(
   A=Subtract(A=Time(), B=ScalarParameter() {group=Color_Shading, parameterName=DeployTime})
   B=ScalarParameter() {defaultValue=0.4, group=Color_Shading, parameterName=DeployTimeThreshold}
  )
  n25 = LinearInterpolate(
   B=ScalarParameter() {defaultValue=7, group=WPO_Animation, parameterName=HighEM}
   Alpha=LinearInterpolate(A=NamedRerouteUsage(), B=Saturate(None=@n24), Alpha=NamedRerouteUsage())
  ) {constA=1, constB=4}
  Divide(
   A=Multiply(A=Add(A=@n18, B=@n22), B=StaticSwitchParameter(True=@n23, False=@n25) {group=PickupOverride, parameterName=AsPickupMesh})
   B=EyeAdaptation().EyeAdaptation
  )
  Multiply()
  n26 = VertexColor()
  NamedRerouteDeclaration(None=@n26.A) {name=VtxAlpha, nodeColor=<rgba 0.5234 0 1 1>}
  n27 = VectorParameter() {group=Color_Shading, parameterName=DarkBlue, defaultValue=<rgba 0.0261 0.2041 0.5573 1>}
  n28 = VectorParameter() {group=Color_Shading, parameterName=LightBlue, defaultValue=<rgba 0 0.653 0.7396 1>}
  n29 = Power(Base=@n17) {constExponent=2.2}
  NamedRerouteDeclaration(None=Multiply(A=NamedRerouteUsage(), B=LinearInterpolate(A=@n27.RGB, B=@n28.RGB, Alpha=@n29))) {name=Diffuse_OUT, nodeColor=<rgba 0 1 0.2734 1>}
  n30 = Frac(None=Length(None=NamedRerouteUsage()) {is=class MaterialExpressionLength})
  n31 = Multiply(A=@n30) {constB=3}
  n32 = Frac(None=@n31)
  NamedRerouteDeclaration(None=LinearInterpolate(A=ScalarParameter() {defaultValue=0.5, parameterName=HashMin}, Alpha=Power(Base=@n32)) {constA=0.2}) {name=SimpleHashNoise, nodeColor=<rgba 0 1 0.9531 1>}
  NamedRerouteDeclaration(None=Transform(None=@n7) {transformSourceType=TRANSFORMSOURCE_Local}) {name=FlattenAnimation, nodeColor=<rgba 0.4297 0 1 1>}
  n33 = Add(A=@n31, B=Multiply(A=Time(), B=ScalarParameter() {defaultValue=0.3, parameterName=DerezSpeed_Quick}))
  NamedRerouteDeclaration(None=LinearInterpolate(A=ScalarParameter() {defaultValue=0.3, parameterName=HashMin_Quick}, Alpha=Frac(None=@n33)) {constA=0.2}) {name=SimpleHashNoise_Quick, nodeColor=<rgba 0 1 0.9531 1>}
  NamedRerouteDeclaration(None=@n32) {name=SimpleRandom, nodeColor=<rgba 0.6641 0 1 1>}
  NamedRerouteDeclaration(None=@n11) {name=FlattenAnimation_World, nodeColor=<rgba 0.4297 0 1 1>}
  n34 = Frac(None=Multiply(A=Time(), B=ScalarParameter() {defaultValue=1, group=WPO_Animation_Vis, parameterName=VisSpeed}))
  n35 = StaticSwitchParameter(True=OneMinus(None=@n34), False=@n34) {group=WPO_Animation_Vis, parameterName=ReverseProgress}
  n36 = LinearInterpolate(
   A=LinearInterpolate(
    A=ScalarParameter() {defaultValue=1, group=WPO_Animation_Control, parameterName=VtxHide_Move}
    B=ScalarParameter() {defaultValue=1, group=WPO_Animation_Control, parameterName=VtxHide_Move_Deploying}
    Alpha=NamedRerouteUsage()
   )
   B=ScalarParameter() {defaultValue=1, group=WPO_Animation_Control, parameterName=VtxHide_Move_Forming}
   Alpha=NamedRerouteDeclaration(None=ScalarParameter() {group=WPO_Animation_Control, parameterName=IsForming}) {name=IsForming, nodeColor=<rgba 0 0.6016 1 1>}
  )
  n37 = NamedRerouteDeclaration(None=Saturate(None=StaticSwitchParameter(True=@n35, False=@n36) {group=WPO_Animation_Vis, parameterName=VisMove})) {name=Move_Alpha, nodeColor=<rgba 1 0.2344 0 1>}
  NamedRerouteDeclaration(None=Multiply(A=@n37, B=ScalarParameter() {defaultValue=2, parameterName=RotateTurn}) {constB=360}) {name=RotateAlpha, nodeColor=<rgba 1 0.2344 0 1>}
  n38 = Constant() {r=0.65}
  NamedRerouteDeclaration(None=Step(Y=@n38, X=NamedRerouteUsage()) {constX=0.1}) {name=CenterAlpha, nodeColor=<rgba 1 0.75 0 1>}
  n39 = ComponentMask(None=TextureCoordinate() {coordinateIndex=1}) {r=true}
  n40 = Subtract(A=Multiply(A=@n37, B=ScalarParameter() {defaultValue=1.2, parameterName=AlphaMulti}) {constB=2}, B=@n39)
  NamedRerouteDeclaration(None=SmoothStep(Value=Saturate(None=@n40)) {constMax=0.5}) {name=CenterProgress, nodeColor=<rgba 1 0.1172 0 1>}
  n41 = NamedRerouteUsage()
  n42 = NamedRerouteUsage()
  n43 = LinearInterpolate(Alpha=NamedRerouteUsage()) {constA=2, constB=2}
  n44 = ComponentMask(None=TextureCoordinate() {coordinateIndex=1}) {r=true}
  n45 = Multiply(
   A=Multiply(A=LinearInterpolate(A=@n41, Alpha=@n42) {constB=0}, B=LinearInterpolate(B=@n43, Alpha=@n44) {constA=1, constB=2})
   B=ScalarParameter() {defaultValue=1, group=WPO_Animation, parameterName=RotateSpeedX}
  )
  NamedRerouteDeclaration(None=@n45) {name=RotateX, nodeColor=<rgba 1 0.7734 0 1>}
  n46 = NamedRerouteUsage()
  n47 = NamedRerouteUsage()
  n48 = LinearInterpolate(Alpha=NamedRerouteUsage()) {constA=2, constB=1.2}
  n49 = ComponentMask(None=TextureCoordinate() {coordinateIndex=1}) {r=true}
  n50 = Multiply(
   A=Multiply(A=LinearInterpolate(A=@n46, Alpha=@n47) {constB=0}, B=LinearInterpolate(B=@n48, Alpha=@n49) {constA=1, constB=2})
   B=ScalarParameter() {defaultValue=1, group=WPO_Animation, parameterName=RotateSpeedY}
  )
  NamedRerouteDeclaration(None=@n50) {name=RotateY, nodeColor=<rgba 0.4531 0 1 1>}
  n51 = NamedRerouteUsage()
  n52 = Multiply(
   A=LinearInterpolate(
    A=ScalarParameter() {defaultValue=5, group=WPO_Animation, parameterName=TestUnit_WPO_Max}
    B=ScalarParameter() {defaultValue=3, group=WPO_Animation, parameterName=TestUnit_WPO_Min}
    Alpha=NamedRerouteUsage()
   )
   B=NamedRerouteUsage()
  )
  n53 = SmoothStep(Value=NamedRerouteUsage()) {constMin=0.9}
  n54 = Multiply(A=ScalarParameter() {defaultValue=4, group=WPO_Animation, parameterName=PulseOffsetWPO}, B=Normalize(VectorInput=NamedRerouteUsage()))
  n55 = Transform(None=Add(A=Multiply(A=@n51, B=@n52), B=Multiply(A=@n53, B=@n54))) {transformSourceType=TRANSFORMSOURCE_Local}
  NamedRerouteDeclaration(None=@n55) {name=IdleWPO, nodeColor=<rgba 1 0 0.2812 1>}
  n56 = Normalize(VectorInput=Transform(None=Constant3Vector() {constant=<rgba 0 0 1 0>}) {transformType=TRANSFORM_Local})
  n57 = NamedRerouteUsage()
  n58 = Cosine(None=@n57)
  n59 = Sine(None=@n57)
  n60 = Constant()
  n61 = MaterialFunctionCall(X=@n58, Y=Multiply(A=@n59) {constB=-1}, Z=@n60) {materialFunction=&MakeFloat3}
  n62 = MaterialFunctionCall(X=@n59, Y=@n58, Z=@n60) {materialFunction=&MakeFloat3}
  n63 = Constant3Vector() {constant=<rgba 0 0 1 0>}
  n64 = Constant()
  n65 = NamedRerouteUsage()
  n66 = Cosine(None=@n65)
  n67 = Sine(None=@n65)
  n68 = Multiply(A=@n67) {constB=-1}
  n69 = Normalize(VectorInput=MaterialFunctionCall(
    VectorToTransform=MaterialFunctionCall(VectorToTransform=@n56, BasisX=@n61.Result, BasisY=@n62.Result, BasisZ=@n63) {materialFunction=&Transform3x3Matrix}.Transformed Vector
    BasisX=Constant3Vector() {constant=<rgba 1 0 0 0>}
    BasisY=MaterialFunctionCall(X=@n64, Y=@n66, Z=@n68) {materialFunction=&MakeFloat3}.Result
    BasisZ=MaterialFunctionCall(X=@n64, Y=@n67, Z=@n66) {materialFunction=&MakeFloat3}.Result
    BasisPosition=Constant3Vector()
   ) {materialFunction=&Transform3x3Matrix}.Transformed Vector)
  NamedRerouteDeclaration(None=StaticSwitchParameter(
    True=Constant3Vector() {constant=<rgba 0 0 1 0>}
    False=Transform(None=@n69) {transformSourceType=TRANSFORMSOURCE_Local, transformType=TRANSFORM_Tangent}
   ) {group=PickupOverride, parameterName=AsPickupMesh}) {name=Normal_OUT, nodeColor=<rgba 1 0.4453 0 1>}
  n70 = Multiply(A=Normalize(VectorInput=@n1), B=@n4)
  NamedRerouteDeclaration(None=Add(
    A=VectorParameter() {group=WPO_Animation_Control, parameterName=FormingAnimationOffsetUnit, defaultValue=<rgba 0 0 4 1>}.RGB
    B=Transform(None=@n70) {transformSourceType=TRANSFORMSOURCE_Local}
   )) {name=FormingAnimation, nodeColor=<rgba 1 0.0938 0 1>}
  n71 = SmoothStep(Max=Multiply(A=@n37) {constB=0.2}, Value=Saturate(None=Subtract(A=@n39, B=@n37))) {constMin=-0.02}
  n72 = Constant() {r=0.35}
  n73 = Saturate(None=Add(A=Subtract(A=@n37, B=@n39), B=Multiply(A=@n37, B=@n72) {constB=0.2}) {constB=0.2})
  n74 = SmoothStep(Max=@n72, Value=@n73) {constMax=0.2}
  NamedRerouteDeclaration(None=LinearInterpolate(A=OneMinus(None=@n71), Alpha=@n74)) {name=FormingProgress, nodeColor=<rgba 0.2187 1 0 1>}
  NamedRerouteDeclaration(None=@n74) {name=ShouldnotMove_Forming, nodeColor=<rgba 0.2656 1 0 1>}
  n75 = Divide(A=Length(None=LocalPosition().XYZ), B=ScalarParameter() {defaultValue=8, group=WPO_Animation, parameterName=GlowLength})
  NamedRerouteDeclaration(None=OneMinus(None=Saturate(None=@n75))) {name=CenterGlow, nodeColor=<rgba 0 1 0.4141 1>}
  n76 = Constant() {r=0.15}
  n77 = Multiply(A=Constant3Vector() {constant=<rgba 1 0 0 0>}, B=Step(Y=NamedRerouteUsage(), X=@n76) {constX=0.1})
  n78 = NamedRerouteUsage()
  n79 = NamedRerouteUsage()
  n80 = Constant() {r=0.25}
  n81 = Multiply(A=Constant3Vector() {constant=<rgba -1 0 0 0>}, B=Multiply(A=Step(Y=@n76, X=@n78) {constX=0.1}, B=Step(Y=@n79, X=@n80) {constX=0.1}))
  n82 = Constant3Vector() {constant=<rgba 0 0 1 0>}
  n83 = Constant() {r=0.35}
  n84 = Multiply(A=Step(Y=@n80, X=NamedRerouteUsage()) {constX=0.1}, B=Step(Y=NamedRerouteUsage(), X=@n83) {constX=0.1})
  n85 = Step(Y=@n83, X=NamedRerouteUsage()) {constX=0.1}
  n86 = Constant() {r=0.45}
  n87 = Step(Y=NamedRerouteUsage(), X=@n86) {constX=0.1}
  n88 = Add(
   A=Add(A=Add(A=@n77, B=@n81), B=Multiply(A=@n82, B=@n84))
   B=Multiply(A=Constant3Vector() {constant=<rgba 0 0 -1 0>}, B=Multiply(A=@n85, B=@n87))
  )
  n89 = NamedRerouteUsage()
  n90 = NamedRerouteUsage()
  n91 = Constant() {r=0.55}
  n92 = Multiply(A=Constant3Vector() {constant=<rgba 0 1 0 0>}, B=Multiply(A=Step(Y=@n86, X=@n89) {constX=0.1}, B=Step(Y=@n90, X=@n91) {constX=0.1}))
  n93 = Constant3Vector() {constant=<rgba 0 -1 0 0>}
  n94 = Multiply(A=Step(Y=@n91, X=NamedRerouteUsage()) {constX=0.1}, B=Step(Y=NamedRerouteUsage(), X=@n38) {constX=0.1})
  NamedRerouteDeclaration(None=Add(A=Add(A=@n88, B=@n92), B=Multiply(A=@n93, B=@n94))) {name=Local_CenterN, nodeColor=<rgba 0.6641 0 1 1>}
  n95 = Add(A=@n31, B=Multiply(A=Time(), B=ScalarParameter() {defaultValue=0.1, parameterName=DerezSpeed}))
  NamedRerouteDeclaration(None=ConstantBiasScale(None=Sine(None=@n95))) {name=SimpleRandomTime, nodeColor=<rgba 0.7109 1 0 1>}
  n96 = ScalarParameter() {defaultValue=2, group=WPO_Animation, parameterName=UV_Tiling}
  n97 = AppendVector(A=Length(None=NamedRerouteUsage()), B=NamedRerouteUsage())
  n98 = AppendVector(
   A=ScalarParameter() {defaultValue=0.2, group=WPO_Animation, parameterName=PanningSpeedX}
   B=ScalarParameter() {defaultValue=0.2, group=WPO_Animation, parameterName=PanningSpeedZ}
  )
  n99 = Time()
  n100 = TextureSample(
   UVs=Add(A=Multiply(A=@n96, B=@n97), B=Multiply(A=@n98, B=@n99))
   Tex=TextureObject() {samplerType=SAMPLERTYPE_Masks, texture=/Game/Global/Textures/Default/Noise/WindNoise_Mask_1}
  )
  n101 = ScalarParameter() {defaultValue=0.4, parameterName=CenterBias}
  n102 = SmoothStep(Value=@n100.R) {constMin=0.3}
  n103 = ScalarParameter() {defaultValue=0.2, parameterName=FrameBias}
  NamedRerouteDeclaration(None=LinearInterpolate(A=Subtract(A=@n100.R, B=@n101) {constB=0.5}, B=Subtract(A=@n102, B=@n103), Alpha=NamedRerouteUsage())) {name=Texture_LocalNoise, nodeColor=<rgba 0 1 0.3438 1>}
  n104 = Multiply(A=Subtract(
    A=MaterialFunctionCall(X=@n26.R, Y=@n26.R, Z=@n26.R) {materialFunction=&MakeFloat3}.Result
    B=Constant3Vector() {constant=<rgba 0.5 0.5 0.5 0>}
   ) {constB=0.5}) {constB=10}
  NamedRerouteDeclaration(None=Add(A=@n104, B=Multiply(B=@n26.A) {constA=255})) {name=LocalWorld_Seed, nodeColor=<rgba 0.8047 1 0 1>}
  n105 = Frac(None=@n95)
  NamedRerouteDeclaration(None=LinearInterpolate(A=ScalarParameter() {defaultValue=0.3, parameterName=HashMin_Cd}, Alpha=Power(Base=@n105)) {constA=0.2}) {name=SimpleHashNoise_Cd, nodeColor=<rgba 0 1 0.9531 1>}
  NamedRerouteDeclaration(None=@n30) {name=LocalNoise, nodeColor=<rgba 1 0.8203 0 1>}
  n106 = NamedRerouteDeclaration(None=@n104) {name=LocalWorldPos, nodeColor=<rgba 1 0 0.4453 1>}
  n107 = Constant()
  NamedRerouteDeclaration(None=If(
    A=Length(None=@n106)
    B=ScalarParameter() {defaultValue=10, group=WPO_Animation, parameterName=FrameLength}
    "A > B"=Constant() {r=1}
    "A == B"=@n107
    "A < B"=@n107
   )) {name=FrameArea, nodeColor=<rgba 0.1016 1 0 1>}
  n108 = Transform(None=Multiply(A=NamedRerouteUsage()) {constB=-1}) {transformSourceType=TRANSFORMSOURCE_Local}
  n109 = WorldPosition()
  n110 = Cosine(None=@n45)
  n111 = Sine(None=@n45)
  n112 = Constant()
  n113 = MaterialFunctionCall(
   VectorToTransform=TransformPosition(None=Add(A=@n108, B=@n109.XYZ)) {transformSourceType=TRANSFORMPOSSOURCE_World}
   BasisX=MaterialFunctionCall(X=@n110, Y=Multiply(A=@n111) {constB=-1}, Z=@n112) {materialFunction=&MakeFloat3}.Result
   BasisY=MaterialFunctionCall(X=@n111, Y=@n110, Z=@n112) {materialFunction=&MakeFloat3}.Result
   BasisZ=Constant3Vector() {constant=<rgba 0 0 1 0>}
   BasisPosition=Constant3Vector()
  ) {materialFunction=&Transform3x3Matrix}
  n114 = Constant3Vector() {constant=<rgba 1 0 0 0>}
  n115 = Constant()
  n116 = Cosine(None=@n50)
  n117 = Sine(None=@n50)
  n118 = MaterialFunctionCall(X=@n115, Y=@n116, Z=Multiply(A=@n117) {constB=-1}) {materialFunction=&MakeFloat3}
  n119 = MaterialFunctionCall(X=@n115, Y=@n117, Z=@n116) {materialFunction=&MakeFloat3}
  n120 = Constant3Vector()
  n121 = LinearInterpolate(
   B=ScalarParameter() {defaultValue=1, group=WPO_Animation, parameterName=ScaleByCenter}
   Alpha=LinearInterpolate(A=NamedRerouteUsage(), B=NamedRerouteUsage(), Alpha=NamedRerouteUsage())
  ) {constA=1}
  n122 = LinearInterpolate(B=ScalarParameter() {defaultValue=2, group=WPO_Animation, parameterName=ScaleByCenter_Forming}, Alpha=NamedRerouteUsage()) {constA=1}
  n123 = NamedRerouteUsage()
  n124 = TransformPosition(None=Multiply(
    A=MaterialFunctionCall(VectorToTransform=@n113.Transformed Vector, BasisX=@n114, BasisY=@n118.Result, BasisZ=@n119.Result, BasisPosition=@n120) {materialFunction=&Transform3x3Matrix}.Transformed Vector
    B=LinearInterpolate(A=@n121, B=@n122, Alpha=@n123)
   )) {transformType=TRANSFORMPOSSOURCE_World}
  n125 = WorldPosition()
  NamedRerouteDeclaration(None=Subtract(A=Subtract(A=@n124, B=@n125.XYZ), B=@n108)) {name=ScalePos, nodeColor=<rgba 1 0.0703 0 1>}
  NamedRerouteUsage()
  n126 = Multiply(A=VectorParameter() {parameterName=PawnWorldPOS, defaultValue=<rgba 0 0 -150 1>}.RGB, B=NamedRerouteUsage())
  n127 = Multiply(
   A=Multiply(A=ScalarParameter() {defaultValue=120, group=WPO_Animation_Control, parameterName=BurstUnit}, B=NamedRerouteUsage())
   B=NamedRerouteUsage()
  )
  n128 = Multiply(A=NamedRerouteUsage(), B=ScalarParameter() {defaultValue=120, group=WPO_Animation_Control, parameterName=BurstUnit_Deploying})
  n129 = NamedRerouteUsage()
  n130 = Subtract(A=@n129, B=ComponentMask(None=TextureCoordinate() {coordinateIndex=1}) {r=true})
  n131 = Saturate(None=Multiply(A=@n129, B=SmoothStep(Value=@n130) {constMax=0.1}))
  n132 = ScalarParameter() {group=WPO_Animation_Control, parameterName=IsDeploying}
  n133 = ScalarParameter() {defaultValue=100, group=WPO_Animation_Control, parameterName=BurstUnit_Forming}
  n134 = NamedRerouteUsage()
  n135 = LinearInterpolate(
   A=LinearInterpolate(
    A=Add(A=@n126, B=@n127)
    B=Multiply(A=@n128, B=@n131)
    Alpha=NamedRerouteDeclaration(None=@n132) {name=IsDeploying, nodeColor=<rgba 1 0 0.4687 1>}
   )
   B=Multiply(A=NamedRerouteUsage(), B=Multiply(A=@n133, B=@n134))
   Alpha=NamedRerouteUsage()
  )
  n136 = Add(A=@n55, B=NamedRerouteUsage())
  n137 = Add(A=ObjectPositionWS(), B=@n135)
  n138 = WorldPosition()
  n139 = ScalarParameter() {group=WPO_Animation_Control, parameterName=VtxHide}
  n140 = TextureCoordinate() {coordinateIndex=1}
  n141 = Saturate(None=Subtract(A=Saturate(None=@n139), B=ComponentMask(None=@n140) {r=true}))
  n142 = ScalarParameter() {group=WPO_Animation_Control, parameterName=VtxHide_Deploying}
  n143 = LinearInterpolate(
   A=LinearInterpolate(A=SmoothStep(Value=@n141) {constMax=0.5}, B=Saturate(None=@n142), Alpha=NamedRerouteUsage())
   B=LinearInterpolate(Alpha=NamedRerouteUsage())
   Alpha=NamedRerouteUsage()
  )
  n144 = StaticSwitchParameter(
   True=NamedRerouteUsage()
   False=LinearInterpolate(
    A=Add(A=@n135, B=@n136)
    B=Subtract(A=@n137, B=@n138.XYZ)
    Alpha=NamedRerouteDeclaration(None=@n143) {name=VtxHide_Alpha, nodeColor=<rgba 0 0.7656 1 1>}
   )
  ) {group=PickupOverride, parameterName=AsPickupMesh}
  Reroute(None=PreviousFrameSwitch(Current Frame=@n144, Previous Frame=@n144))
  Saturate(None=Sign(None=@n73))
 ]
}
