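# Test Default Lit material for a magic cube movement effect, combining animated blue shading with noise-driven vertex offsets.
# Vertex color supplies a seed, frame mask and alpha, while a panning wind-noise texture gets separate center and frame biases.
# Step bands decode a scalar into six local axis directions, scaled and transformed as an offset.
# Sine-animated noise blends two blues, alongside a time-scrolled hash brightness and center emissive blend.
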Material /CubeSurfing/Gameplay/Material/M_MagicCubeMovement_FX_TEST
graph {
 properties {bAutomaticallySetUsageInEditor=false, bUsedWithNiagaraRibbons=true, usedShadingModels=Default Lit}
 nodes [
  n1 = ScalarParameter() {defaultValue=2, group=WPO_Animation, parameterName=UV_Tiling}
  n2 = AppendVector(A=ComponentMask(None=NamedRerouteUsage()) {r=true}, B=NamedRerouteUsage())
  n3 = AppendVector(
   A=ScalarParameter() {defaultValue=0.2, group=WPO_Animation, parameterName=PanningSpeedX}
   B=ScalarParameter() {defaultValue=0.2, group=WPO_Animation, parameterName=PanningSpeedZ}
  )
  n4 = Time()
  n5 = TextureSample(
   UVs=Add(A=Multiply(A=@n1, B=@n2), B=Multiply(A=@n3, B=@n4))
   Tex=TextureObject() {samplerType=SAMPLERTYPE_Masks, texture=/Game/Global/Textures/Default/Noise/WindNoise_Mask_1}
  )
  n6 = Subtract(A=@n5.R, B=ScalarParameter() {defaultValue=0.4, parameterName=CenterBias}) {constB=0.5}
  Add(A=@n6)
  Constant()
  Constant() {r=10}
  Constant() {r=0.5}
  n7 = VectorParameter() {group=Shading_TechArt, parameterName=CenterEM, channelNames {a=EM_Intensity, b="", g="", r=""}, defaultValue=<rgba 0.033 0.0594 0.066 10>}
  LinearInterpolate(A=NamedRerouteUsage(), B=Multiply(A=@n7, B=@n7), Alpha=NamedRerouteUsage())
  n8 = VertexColor()
  NamedRerouteDeclaration(None=@n8.A) {name=VtxAlpha, nodeColor=<rgba 0.5234 0 1 1>}
  n9 = VectorParameter() {parameterName=DarkBlue, defaultValue=<rgba 0.0261 0.2041 0.5573 1>}
  n10 = VectorParameter() {parameterName=LightBlue, defaultValue=<rgba 0 0.653 0.7396 1>}
  n11 = Multiply(A=NamedRerouteUsage(), B=ScalarParameter() {defaultValue=1, parameterName=LocalNoise_Intensity}) {constB=0.3}
  n12 = Multiply(A=Time())
  n13 = ConstantBiasScale(None=Sine(None=Add(A=@n11, B=@n12)))
  n14 = Power(Base=Saturate(None=Add(A=@n13) {constB=0.2})) {constExponent=2.2}
  NamedRerouteDeclaration(None=Multiply(A=NamedRerouteUsage(), B=LinearInterpolate(A=@n9.RGB, B=@n10.RGB, Alpha=@n14))) {name=Diffuse_OUT, nodeColor=<rgba 0 1 0.2734 1>}
  n15 = Constant() {r=0.65}
  NamedRerouteDeclaration(None=Step(Y=@n15, X=NamedRerouteUsage()) {constX=0.1}) {name=CenterAlpha, nodeColor=<rgba 1 0.75 0 1>}
  n16 = Constant3Vector() {constant=<rgba 1 0 0 0>}
  n17 = Constant() {r=0.15}
  n18 = Step(Y=NamedRerouteUsage(), X=@n17) {constX=0.1}
  n19 = Constant3Vector() {constant=<rgba -1 0 0 0>}
  n20 = Constant() {r=0.25}
  n21 = Multiply(A=Step(Y=@n17, X=NamedRerouteUsage()) {constX=0.1}, B=Step(Y=NamedRerouteUsage(), X=@n20) {constX=0.1})
  n22 = Step(Y=@n20, X=NamedRerouteUsage()) {constX=0.1}
  n23 = Constant() {r=0.35}
  n24 = Step(Y=NamedRerouteUsage(), X=@n23) {constX=0.1}
  n25 = Add(
   A=Add(A=Multiply(A=@n16, B=@n18), B=Multiply(A=@n19, B=@n21))
   B=Multiply(A=Constant3Vector() {constant=<rgba 0 0 1 0>}, B=Multiply(A=@n22, B=@n24))
  )
  n26 = NamedRerouteUsage()
  n27 = NamedRerouteUsage()
  n28 = Constant() {r=0.45}
  n29 = Multiply(A=Constant3Vector() {constant=<rgba 0 0 -1 0>}, B=Multiply(A=Step(Y=@n23, X=@n26) {constX=0.1}, B=Step(Y=@n27, X=@n28) {constX=0.1}))
  n30 = Constant3Vector() {constant=<rgba 0 1 0 0>}
  n31 = Constant() {r=0.55}
  n32 = Multiply(A=Step(Y=@n28, X=NamedRerouteUsage()) {constX=0.1}, B=Step(Y=NamedRerouteUsage(), X=@n31) {constX=0.1})
  n33 = Step(Y=@n31, X=NamedRerouteUsage()) {constX=0.1}
  n34 = Step(Y=NamedRerouteUsage(), X=@n15) {constX=0.1}
  n35 = Add(
   A=Add(A=Add(A=@n25, B=@n29), B=Multiply(A=@n30, B=@n32))
   B=Multiply(A=Constant3Vector() {constant=<rgba 0 -1 0 0>}, B=Multiply(A=@n33, B=@n34))
  )
  NamedRerouteDeclaration(None=@n35) {name=Local_CenterN, nodeColor=<rgba 0.6641 0 1 1>}
  n36 = SmoothStep(Value=@n5.R) {constMin=0.3}
  n37 = ScalarParameter() {defaultValue=0.2, parameterName=FrameBias}
  NamedRerouteDeclaration(None=LinearInterpolate(A=@n6, B=Subtract(A=@n36, B=@n37), Alpha=NamedRerouteUsage())) {name=Texture_LocalNoise, nodeColor=<rgba 0 1 0.3438 1>}
  n38 = MaterialFunctionCall(X=@n8.R, Y=@n8.R, Z=@n8.R) {materialFunction=&MakeFloat3}
  n39 = Constant3Vector() {constant=<rgba 0.5 0.5 0.5 0>}
  NamedRerouteDeclaration(None=Multiply(A=Subtract(A=@n38.Result, B=@n39) {constB=0.5}) {constB=255}) {name=LocalWorld_Seed, nodeColor=<rgba 0.8047 1 0 1>}
  n40 = Frac(None=Length(None=NamedRerouteUsage()) {is=class MaterialExpressionLength})
  n41 = Time()
  n42 = ScalarParameter() {defaultValue=0.1, parameterName=DerezSpeed}
  n43 = Frac(None=Add(A=Multiply(A=@n40) {constB=3}, B=Multiply(A=@n41, B=@n42)))
  NamedRerouteDeclaration(None=LinearInterpolate(A=ScalarParameter() {defaultValue=0.3, parameterName=HashMin_Cd}, Alpha=Power(Base=@n43)) {constA=0.2}) {name=SimpleHashNoise_Cd, nodeColor=<rgba 0 1 0.9531 1>}
  NamedRerouteDeclaration(None=@n40) {name=LocalNoise, nodeColor=<rgba 1 0.8203 0 1>}
  NamedRerouteDeclaration(None=@n8.G) {name=FrameArea, nodeColor=<rgba 0.1016 1 0 1>}
  NamedRerouteUsage()
  Power(Base=@n5.R) {constExponent=2.2}
  ScalarParameter() {defaultValue=20, group=WPO_Animation, parameterName=WorldAligned_Unit}
  n44 = LinearInterpolate(
   A=ScalarParameter() {defaultValue=5, group=WPO_Animation, parameterName=TestUnit_WPO_Max}
   B=ScalarParameter() {defaultValue=3, group=WPO_Animation, parameterName=TestUnit_WPO_Min}
   Alpha=NamedRerouteUsage()
  )
  n45 = NamedRerouteUsage()
  Transform(None=Multiply(A=@n35, B=Multiply(A=@n44, B=@n45))) {transformSourceType=TRANSFORMSOURCE_Local}
 ]
}
